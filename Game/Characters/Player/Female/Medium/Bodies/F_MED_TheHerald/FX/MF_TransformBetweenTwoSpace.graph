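# Material function in The Herald character's FX folder that re-orients a vector from one normal/tangent frame into another, labelled as pre-skinned to skinned space.
# It takes target and source normals and tangents plus the vector, feeding one Custom HLSL node that derives each bitangent by cross product.
# The node multiplies the target basis by the transposed source basis, used as its inverse, which assumes orthonormal inputs, then applies the result to the vector.
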MaterialFunction /Game/Characters/Player/Female/Medium/Bodies/F_MED_TheHerald/FX/MF_TransformBetweenTwoSpace
graph {nodes [
  MaterialExpressionFunctionInput_0 = FunctionInput() {inputName=TargetSpace_Normal}
  MaterialExpressionFunctionInput_1 = FunctionInput() {inputName=TargetSpace_Tangent, sortPriority=1}
  MaterialExpressionFunctionInput_2 = FunctionInput() {inputName=FromSpace_Normal, sortPriority=2}
  MaterialExpressionFunctionInput_3 = FunctionInput() {inputName=FromSpace_Tangent, sortPriority=3}
  MaterialExpressionFunctionInput_4 = FunctionInput() {inputName=InputVector, sortPriority=4}
  n1 = Custom!(
   TargetSpaceN=@MaterialExpressionFunctionInput_0
   TargetSpaceT=@MaterialExpressionFunctionInput_1
   FromSpaceN=@MaterialExpressionFunctionInput_2
   FromSpaceT=@MaterialExpressionFunctionInput_3
   InputVector=@MaterialExpressionFunctionInput_4
  ) {
   code="float3 TargetSpaceB = cross(TargetSpaceT,TargetSpaceN);\r\nfloat3 FromSpaceB = cross(FromSpaceT,FromSpaceN);\r\n\r\nfloat3x3 TargetSpaceMat = {\r\n    TargetSpaceT.x,TargetSpaceB.x,TargetSpaceN.x,\r\n    TargetSpaceT.y,TargetSpaceB.y,TargetSpaceN.y,\r\n    TargetSpaceT.z,TargetSpaceB.z,TargetSpaceN.z\r\n};\r\nfloat3x3 FromSpaceMatInverse = {\r\n    FromSpaceT.x,FromSpaceT.y,FromSpaceT.z,\r\n    FromSpaceB.x,FromSpaceB.y,FromSpaceB.z,\r\n    FromSpaceN.x,FromSpaceN.y,FromSpaceN.z\r\n};\r\nreturn mul(mul(TargetSpaceMat,FromSpaceMatInverse),InputVector);"
   description=PreSkinSpaceToSkinedSpace
   inputs [
    {
     inputName=TargetSpaceN
     input {expression=&self:MaterialExpressionFunctionInput_0}
    }
    {
     inputName=TargetSpaceT
     input {expression=&self:MaterialExpressionFunctionInput_1}
    }
    {
     inputName=FromSpaceN
     input {expression=&self:MaterialExpressionFunctionInput_2}
    }
    {
     inputName=FromSpaceT
     input {expression=&self:MaterialExpressionFunctionInput_3}
    }
    {
     inputName=InputVector
     input {expression=&self:MaterialExpressionFunctionInput_4}
    }
   ]
  }
  FunctionOutput(None=@n1)
 ]}
