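# Core material function that simulates feather bending on The Herald player body via vertex shader, outputting WPO, an adjusted normal and two rotation strengths.
# Decodes per-feather pivot, axes, original normal and masks from vertex color, UVs and two textures, remapping them from pre-skinned to current tangent space.
# A rigidity-scaled force vector and per-feather randomized sine wind drive two RotateAboutAxis rotations, with one angle clamped.
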MaterialFunction /Game/Characters/Player/Female/Medium/Bodies/F_MED_TheHerald/FX/MF_FeatherSimulation_Core
graph {nodes [
  FunctionOutput(None=LinearInterpolate(A=VertexNormalWS(), B=NamedRerouteUsage(), Alpha=NamedRerouteUsage())) {outputName=Normal}
  FunctionOutput(None=Multiply(A=NamedRerouteUsage(), B=NamedRerouteUsage())) {outputName=WPO}
  n1 = NamedRerouteUsage()
  n2 = NamedRerouteUsage()
  n3 = Multiply(A=DotProduct(A=NamedRerouteUsage(), B=NamedRerouteUsage())) {constB=-1}
  n4 = FunctionInput() {
   bUsePreviewValueAsDefault=true
   inputName=ScaleForceAlignFeatherNormal
   inputType=FunctionInput_Scalar
   sortPriority=251
   previewValue=<0.7 0 0 1>
  }
  n5 = Add(A=Saturate(None=DotProduct(A=@n1, B=@n2)), B=Multiply(A=Saturate(None=@n3), B=@n4))
  n6 = FunctionInput() {inputName=AdditionalSpin, inputType=FunctionInput_Scalar, sortPriority=251}
  n7 = FunctionInput() {bUsePreviewValueAsDefault=true, inputName=MaxFeatherRot, inputType=FunctionInput_Scalar, sortPriority=251, previewValue=<0.2 0 0 1>}
  n8 = Clamp(None=Multiply(A=NamedRerouteUsage(), B=Add(A=@n5, B=@n6)), Max=@n7) {maxDefault=0.1}
  FunctionOutput(None=@n8) {outputName=RotX_StrengthOut}
  n9 = TextureCoordinate() {coordinateIndex=3}
  n10 = Add(
   A=FunctionInput() {inputName=WindTime, inputType=FunctionInput_Scalar, sortPriority=250}
   B=Multiply(
    A=MaterialFunctionCall(Float2=@n9) {materialFunction=&BreakOutFloat2Components}.G
    B=FunctionInput() {inputName=WindRandomScale, inputType=FunctionInput_Scalar, sortPriority=251}
   ) {constB=10}
  )
  n11 = Multiply(A=NamedRerouteUsage()) {constB=-1}
  n12 = NamedRerouteUsage()
  n13 = Add(
   A=Multiply(A=Sine(None=@n10), B=FunctionInput() {inputName=WindStrength, inputType=FunctionInput_Scalar, sortPriority=251})
   B=Multiply(A=NamedRerouteUsage(), B=DotProduct(A=@n11, B=@n12))
  )
  FunctionOutput(None=@n13) {outputName=RotY_StrengthOut}
  NamedRerouteDeclaration(None=Transform(None=VertexNormalWS()) {transformSourceType=TRANSFORMSOURCE_World, transformType=TRANSFORM_Local}) {name=TargetSpace_N, nodeColor=<rgba 0 1 0.2734 1>}
  NamedRerouteDeclaration(None=Transform(None=VertexTangentWS()) {transformSourceType=TRANSFORMSOURCE_World, transformType=TRANSFORM_Local}) {name=TargetSpace_T, nodeColor=<rgba 0.0535 1 0 1>}
  NamedRerouteDeclaration(None=CrossProduct(A=NamedRerouteUsage(), B=NamedRerouteUsage())) {name=FeatherData_BiNormal, nodeColor=<rgba 0 0.2031 1 1>}
  n14 = NamedRerouteUsage()
  n15 = NamedRerouteUsage()
  n16 = WorldPosition() {worldPositionShaderOffset=WPT_ExcludeAllShaderOffsets}
  n17 = NamedRerouteUsage()
  n18 = NamedRerouteUsage()
  n19 = WorldPosition() {worldPositionShaderOffset=WPT_ExcludeAllShaderOffsets}
  NamedRerouteDeclaration(None=Add(
    A=RotateAboutAxis(NormalizedRotationAxis=@n14, RotationAngle=@n8, PivotPoint=@n15, Position=@n16.XYZ)
    B=RotateAboutAxis(NormalizedRotationAxis=@n17, RotationAngle=@n13, PivotPoint=@n18, Position=@n19.XYZ)
   )) {name=CalcOutPut_WPO, nodeColor=<rgba 0.8516 0 1 1>}
  n20 = Add(
   A=RotateAboutAxis(NormalizedRotationAxis=NamedRerouteUsage(), RotationAngle=@n8, PivotPoint=Constant(), Position=NamedRerouteUsage())
   B=RotateAboutAxis(NormalizedRotationAxis=NamedRerouteUsage(), RotationAngle=@n13, PivotPoint=Constant(), Position=NamedRerouteUsage())
  )
  n21 = NamedRerouteUsage()
  NamedRerouteDeclaration(None=VertexInterpolator(VS=Add(A=@n20, B=@n21)).PS) {name=CalcOutPut_PixelNormalWS, nodeColor=<rgba 0.8516 0 1 1>}
  n22 = MaterialFunctionCall(Float2=TextureCoordinate() {coordinateIndex=1}) {materialFunction=&BreakOutFloat2Components}
  n23 = Constant()
  NamedRerouteDeclaration(None=If(A=Add(A=@n22.R, B=@n22.R), "A > B"=@n23, "A == B"=@n23, "A < B"=Constant() {r=1}) {constB=2}) {name=FeatherData_FeatherEffectMask, nodeColor=<rgba 1 0 0.6562 1>}
  NamedRerouteDeclaration(None=PreSkinnedNormal()) {name=FromSpace_N, nodeColor=<rgba 0 1 0.2734 1>}
  NamedRerouteDeclaration(None=Constant3Vector() {constant=<rgba 0 0 1 0>}) {name=FromSpace_T, nodeColor=<rgba 0.0535 1 0 1>}
  n24 = VertexColor()
  n25 = Multiply(A=AppendVector(A=AppendVector(A=@n24.R, B=@n24.R), B=@n24.G)) {constB=2}
  n26 = FunctionInput() {bUsePreviewValueAsDefault=true, inputName=FeatherSize, inputType=FunctionInput_Scalar, sortPriority=200, previewValue=<10 0 0 1>}
  n27 = MaterialFunctionCall!(
   TargetSpace_Normal=NamedRerouteUsage()
   TargetSpace_Tangent=NamedRerouteUsage()
   FromSpace_Normal=NamedRerouteUsage()
   FromSpace_Tangent=NamedRerouteUsage()
   InputVector=Multiply(A=Subtract(A=@n25), B=@n26) {constB=10}
  ) {materialFunction=&/MF_TransformBetweenTwoSpace}
  NamedRerouteDeclaration(None=Add(
    A=Transform(None=@n27.Result) {transformSourceType=TRANSFORMSOURCE_Local}
    B=WorldPosition() {worldPositionShaderOffset=WPT_ExcludeAllShaderOffsets}.XYZ
   )) {name=FeatherData_Pivot, nodeColor=<rgba 1 0.4922 0 1>}
  n28 = MaterialFunctionCall(Float2=TextureCoordinate() {coordinateIndex=2}) {materialFunction=&BreakOutFloat2Components}
  n29 = TextureCoordinate() {coordinateIndex=3}
  n30 = Multiply(A=MaterialFunctionCall(
    X=@n28.R
    Y=MaterialFunctionCall(Float2=@n29) {materialFunction=&BreakOutFloat2Components}.R
    Z=OneMinus(None=@n28.G)
   ) {materialFunction=&MakeFloat3}.Result) {constB=2}
  n31 = MaterialFunctionCall(
   TargetSpace_Normal=NamedRerouteUsage()
   TargetSpace_Tangent=NamedRerouteUsage()
   FromSpace_Normal=NamedRerouteUsage()
   FromSpace_Tangent=NamedRerouteUsage()
   InputVector=Normalize(VectorInput=Subtract(A=@n30))
  ) {materialFunction=&MF_TransformBetweenTwoSpace}
  NamedRerouteDeclaration(None=Transform(None=Normalize(VectorInput=@n31.Result)) {transformSourceType=TRANSFORMSOURCE_Local}) {name=FeatherData_OriginalVertexNormal, nodeColor=<rgba 1 0 0.3516 1>}
  n32 = TextureCoordinate() {coordinateIndex=1}
  n33 = TextureSample(UVs=@n32, Tex=FunctionInput() {inputName=FeatherTangentTexture, inputType=FunctionInput_Texture2D}) {samplerType=SAMPLERTYPE_LinearColor}
  n34 = MaterialFunctionCall(Float3=Add(A=Multiply(A=@n33.RGB) {constB=2}) {constB=-1}) {materialFunction=&BreakOutFloat3Components}
  n35 = MaterialFunctionCall(
   TargetSpace_Normal=NamedRerouteUsage()
   TargetSpace_Tangent=NamedRerouteUsage()
   FromSpace_Normal=NamedRerouteUsage()
   FromSpace_Tangent=NamedRerouteUsage()
   InputVector=MaterialFunctionCall(X=@n34.R, Y=@n34.R, Z=@n34.R) {materialFunction=&MakeFloat3}.Result
  ) {materialFunction=&MF_TransformBetweenTwoSpace}
  NamedRerouteDeclaration(None=Transform(None=Normalize(VectorInput=@n35.Result)) {transformSourceType=TRANSFORMSOURCE_Local}) {name=FeatherData_Tangent, nodeColor=<rgba 0.5506 0 1 1>}
  n36 = TextureSample(UVs=@n32, Tex=FunctionInput() {inputName=FeatherNormalTexture, inputType=FunctionInput_Texture2D}) {samplerType=SAMPLERTYPE_LinearColor}
  n37 = MaterialFunctionCall(Float3=Add(A=Multiply(A=@n36.RGB) {constB=2}) {constB=-1}) {materialFunction=&BreakOutFloat3Components}
  n38 = MaterialFunctionCall(
   TargetSpace_Normal=NamedRerouteUsage()
   TargetSpace_Tangent=NamedRerouteUsage()
   FromSpace_Normal=NamedRerouteUsage()
   FromSpace_Tangent=NamedRerouteUsage()
   InputVector=MaterialFunctionCall(X=@n37.R, Y=@n37.R, Z=@n37.R) {materialFunction=&MakeFloat3}.Result
  ) {materialFunction=&MF_TransformBetweenTwoSpace}
  NamedRerouteDeclaration(None=Transform(None=Normalize(VectorInput=@n38.Result)) {transformSourceType=TRANSFORMSOURCE_Local}) {name=FeatherData_Normal, nodeColor=<rgba 1 0 0.9479 1>}
  n39 = FunctionInput() {inputName=Force, sortPriority=10}
  n40 = Divide(
   A=MaterialFunctionCall(Vector 3=@n39) {materialFunction=/Engine/Functions/Engine_MaterialFunctions02/Utility/VectorLength}.V3 Length
   B=FunctionInput() {inputName=Rigidity, inputType=FunctionInput_Scalar, sortPriority=11}
  )
  NamedRerouteDeclaration(None=Multiply(A=Normalize(VectorInput=@n39), B=Saturate(None=@n40))) {name=InputData_Force, nodeColor=<rgba 1 0 0.9844 1>}
  n41 = MaterialFunctionCall(Float3=FunctionInput() {inputName=BendMask_Offset/Mul/Pow, sortPriority=100}) {materialFunction=&BreakOutFloat3Components}
  n42 = Multiply(A=Add(A=VertexColor().A, B=@n41.R), B=@n41.G)
  NamedRerouteDeclaration(None=Power(Base=Saturate(None=@n42), Exp=@n41.B)) {name=FeatherData_BendMask, nodeColor=<rgba 0.7578 0 1 1>}
 ]}
