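# Material function for TheHerald body FX that outputs a view-facing mask from a fake spherical normal on a skeletal mesh.
# Builds the normal from pre-skinned position relative to a bounds-derived, offsettable origin, then rotates it by a quaternion unpacked from a float3, with two components packed into B.
# Dots it with the camera vector, applies Power, and blends via DisableSim to a saturated facing term around DisableSim_HeadCenter.
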MaterialFunction /Game/Characters/Player/Female/Medium/Bodies/F_MED_TheHerald/FX/MF_SkelMeshSphereNormal
graph {nodes [
  n1 = Subtract(A=CameraPositionWS(), B=WorldPosition() {worldPositionShaderOffset=WPT_ExcludeAllShaderOffsets}.XYZ)
  n2 = VertexInterpolator(VS=PreSkinnedPosition())
  n3 = PreSkinnedLocalBounds()
  n4 = Add(A=Divide(A=Add(A=@n3.Min, B=@n3.Min)), B=FunctionInput() {bUsePreviewValueAsDefault=true, inputName=OffsetLS})
  n5 = MaterialFunctionCall(Float3=FunctionInput() {inputName=QuatV3}) {materialFunction=&BreakOutFloat3Components}
  n6 = AppendVector(A=@n5.R, B=@n5.R)
  n7 = Add(A=Multiply(A=Frac(None=@n5.B)) {constB=2}) {constB=-1}
  n8 = Multiply(A=Divide(A=Floor(None=@n5.B)) {constB=1000}) {constB=2}
  n9 = MaterialFunctionCall!(
   Vector=Normalize(VectorInput=Subtract(A=@n2.PS, B=@n4))
   Quat=AppendVector(A=AppendVector(A=@n6, B=@n7), B=Add(A=@n8) {constB=-1})
   Invert Quaternion=StaticBool()
  ) {materialFunction=/Engine/Functions/Engine_MaterialFunctions03/Math/MultiplyVectorWithQuaternion}
  n10 = FunctionInput() {bUsePreviewValueAsDefault=true, inputName=Power, inputType=FunctionInput_Scalar, previewValue=<3 0 0 1>}
  n11 = Power(Base=DotProduct(A=Normalize(VectorInput=@n1), B=Transform(None=@n9.Result) {transformSourceType=TRANSFORMSOURCE_Local}), Exp=@n10)
  n12 = FunctionInput() {inputName=DisableSim, inputType=FunctionInput_Scalar}
  n13 = OneMinus(None=@n12)
  n14 = Subtract(A=CameraPositionWS(), B=WorldPosition() {worldPositionShaderOffset=WPT_ExcludeAllShaderOffsets}.XYZ)
  n15 = Subtract(A=WorldPosition() {worldPositionShaderOffset=WPT_ExcludeAllShaderOffsets}.XYZ, B=FunctionInput() {inputName=DisableSim_HeadCenter})
  n16 = Saturate(None=DotProduct(A=Normalize(VectorInput=@n14), B=Normalize(VectorInput=@n15)))
  FunctionOutput(None=Add(A=Multiply(A=@n11, B=@n13) {constB=0}, B=Multiply(A=@n16, B=@n12)))
  n17 = VertexInterpolator(VS=PreSkinnedPosition())
  n18 = PreSkinnedLocalBounds()
  n19 = Divide(A=Add(A=@n18.Min, B=@n18.Min))
  Normalize(VectorInput=Transform(None=Subtract(A=@n17.PS, B=@n19)) {transformSourceType=TRANSFORMSOURCE_Local})
 ]}
