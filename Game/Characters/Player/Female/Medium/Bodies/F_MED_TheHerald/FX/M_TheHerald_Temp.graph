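# Temporary skeletal-mesh material for The Herald, building an animated, glowing petal colour with feather-simulated normals.
# A three-colour gradient, driven by Fresnel or a spherical skinned-mesh mask, blends with its hue-shifted, desaturated version through two panning noise textures.
# Inverted vertex alpha hue-shifts and dims roots, then a petal detail map modulates brightness into an emissive reroute.
# MF_FeatherSimulation_Core supplies only the VertexNormal reroute; the WPO reroute holds a constant.
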Material /Game/Characters/Player/Female/Medium/Bodies/F_MED_TheHerald/FX/M_TheHerald_Temp
graph {
 properties {bUsedWithMeshDeformer=true, bUsedWithSkeletalMesh=true, pixelDepthOffsetMode=PDOM_Legacy, usedShadingModels=Default Lit}
 nodes [
  Constant()
  Constant() {r=1}
  MaterialFunctionCall(Normal=NamedRerouteUsage(), Flatness=ScalarParameter() {parameterName=FlattenNormal}) {materialFunction=&FlattenNormal}
  n1 = OneMinus(None=VertexColor().A)
  n2 = ScalarParameter() {defaultValue=-0.15, parameterName=Test_HueShift}
  n3 = ScalarParameter() {defaultValue=0.9, parameterName=HueShift}
  n4 = TextureSampleParameter2D() {
   parameterName=NormalMap
   samplerType=SAMPLERTYPE_Normal
   texture=/Game/Characters/Player/Female/Medium/Bodies/F_MED_TheHerald/Textures/T_F_MED_TheHerald_FaceAcc_N
  }
  n5 = Fresnel(Normal=@n4.RGB) {baseReflectFraction=0, exponent=1}
  n6 = MaterialFunctionCall!(
   QuatV3=VectorParameter() {parameterName=Quat}.RGB
   Power=ScalarParameter() {defaultValue=3, parameterName=SphereMaskExp}
   DisableSim=Constant()
   DisableSim_HeadCenter=Constant3Vector()
   OffsetLS=Constant3Vector()
  ) {materialFunction=&/MF_SkelMeshSphereNormal}
  n7 = OneMinus(None=@n6.Result)
  n8 = ScalarParameter() {parameterName=UseSphericalMaskToBlendColor}
  n9 = MaterialFunctionCall(
   A=VectorParameter() {parameterName=ColorA, defaultValue=<rgba 0.5 0.1514 0.1066 1>}.RGB
   B=VectorParameter() {parameterName=ColorB, defaultValue=<rgba 0.5573 0.0448 0.0837 1>}.RGB
   C=VectorParameter() {parameterName=ColorC, defaultValue=<rgba 0.2656 0.0169 0.1034 1>}.RGB
   Alpha=Power(Base=LinearInterpolate(A=@n5, B=@n7, Alpha=@n8), Exp=ScalarParameter() {defaultValue=0.2, parameterName=ColorLerpExp, sliderMin=0.01}) {constExponent=0.5}
  ) {materialFunction=&3ColorBlend}
  n10 = Power(Base=@n9.Result, Exp=ScalarParameter() {defaultValue=1.7, parameterName=HueShfitExp})
  n11 = Divide(A=Fresnel() {baseReflectFraction=0, exponent=1}, B=ScalarParameter() {defaultValue=12, parameterName=Param_1})
  n12 = VertexInterpolator(VS=RotateAboutAxis(
    NormalizedRotationAxis=Constant3Vector() {constant=<rgba 1 0 1 0>}
    RotationAngle=Constant() {r=0.5}
    PivotPoint=PreSkinnedPosition()
    Position=Constant3Vector()
   ))
  n13 = Add(A=ComponentMask(None=@n12.PS) {r=true}, B=Multiply(A=Time(), B=ScalarParameter() {defaultValue=25, parameterName=PanSpeedA}))
  n14 = ScalarParameter() {defaultValue=150, parameterName=TileA}
  n15 = Add(A=ComponentMask(None=@n12.PS) {g=true}, B=Multiply(A=Time(), B=ScalarParameter() {defaultValue=15, parameterName=PanSpeedB}))
  n16 = ScalarParameter() {defaultValue=120, parameterName=TileB}
  n17 = AppendVector(A=Frac(None=Divide(A=@n13, B=@n14) {constB=2000}), B=Frac(None=Divide(A=@n15, B=@n16) {constB=2000}))
  n18 = TextureSample(UVs=Add(A=@n11, B=@n17)) {texture=/Game/Global/Textures/Default/Noise/TankMoteNoise}
  n19 = TextureSample(UVs=Divide(A=Add(A=@n11, B=@n17)) {constB=0.5}) {samplerType=SAMPLERTYPE_LinearGrayscale, texture=/Game/Global/Textures/Default/Noise/T_SwirlNoise_N_height}
  n20 = ScalarParameter() {defaultValue=2, parameterName=DetailNosieEmissve}
  n21 = Multiply(A=@n20, B=@n6.Result)
  n22 = ScalarParameter() {parameterName=UseSphericalMaskToBlendColor}
  n23 = LinearInterpolate(
   A=Desaturation(
    None=MaterialFunctionCall(Hue Shift Percentage=@n3, Texture=@n10) {materialFunction=&HueShift}.Result
    Fraction=ScalarParameter() {defaultValue=0.5, parameterName=Desaturation}
   )
   B=@n9.Result
   Alpha=Multiply(A=Multiply(A=@n18.R, B=@n19.R), B=LinearInterpolate(A=@n20, B=@n21, Alpha=@n22)) {constB=2}
  )
  n24 = ScalarParameter() {defaultValue=1, parameterName=Emissive}
  n25 = ScalarParameter() {defaultValue=-0.5, parameterName=Test_RootScale}
  n26 = Multiply(
   A=MaterialFunctionCall(Hue Shift Percentage=Multiply(A=@n1, B=@n2), Texture=Multiply(A=@n23, B=@n24)) {materialFunction=&HueShift}.Result
   B=Add(A=Multiply(A=@n1, B=@n25))
  )
  n27 = TextureSampleParameter2D() {parameterName=PetalDetailMap, texture=/Game/Characters/Player/Female/Medium/Bodies/F_MED_TheHerald/Textures/T_F_MED_TheHerald_FaceAcc_E}
  n28 = LinearInterpolate(
   A=Multiply(A=@n26, B=ScalarParameter() {defaultValue=1.2, parameterName=PetalDetailLow})
   B=Multiply(A=@n26, B=ScalarParameter() {defaultValue=1.5, parameterName=PetalDetailHigh}) {constB=2}
   Alpha=Power(Base=@n27.R, Exp=ScalarParameter() {defaultValue=3, parameterName=PetalDetailPower}) {constExponent=4}
  )
  MaterialFunctionCall(Emissive=@n28) {materialFunction=&MF_FX_EmissiveAdjustments}
  n29 = TextureSampleParameter2D() {
   parameterName=NormalMap
   samplerType=SAMPLERTYPE_Normal
   texture=&T_F_MED_TheHerald_FaceAcc_N
  }
  MaterialFunctionCall(BaseNormal=@n29.RGB, AdditionalNormal=NamedRerouteUsage()) {materialFunction=&BlendAngleCorrectedNormals}
  NamedRerouteDeclaration(None=Constant()) {name=WPO, nodeColor=<rgba 0 1 0.4375 1>}
  n30 = TextureObject() {
   samplerType=SAMPLERTYPE_LinearColor
   texture=/Game/Characters/Player/Female/Medium/Bodies/F_MED_TheHerald/FX/HairMaterial/T_TheHerald_FD_Tangent
  }
  n31 = TextureObject() {
   samplerType=SAMPLERTYPE_LinearColor
   texture=/Game/Characters/Player/Female/Medium/Bodies/F_MED_TheHerald/FX/HairMaterial/T_TheHerald_FD_Normal
  }
  n32 = Time() {bOverride_Period=true, period=9999}
  n33 = ScalarParameter() {defaultValue=1.2, group=FeatherSim, parameterName=WindSpeed}
  NamedRerouteDeclaration(None=MaterialFunctionCall!(
    FeatherTangentTexture=@n30
    FeatherNormalTexture=@n31
    Force=VectorParameter() {group=FeatherSim, parameterName=Feather_Force, defaultValue=<rgba 0 0 0 1>}.RGB
    Rigidity=ScalarParameter() {defaultValue=400, group=FeatherSim, parameterName=Feather_Rigidity}
    BendMask_Offset/Mul/Pow=Constant3Vector() {constant=<rgba 0 1 1 0>}
    FeatherSize=ScalarParameter() {defaultValue=3, group=FeatherSim, parameterName="FeatherSize (S)"}
    WindTime=Multiply(A=@n32, B=@n33)
    WindRandomScale=ScalarParameter() {defaultValue=0.3, group=FeatherSim, parameterName=WindRandomScale}
    AdditionalSpin=ScalarParameter() {group=FeatherSim, parameterName=AddSpin}
    ScaleForceAlignFeatherNormal
    MaxFeatherRot=ScalarParameter() {defaultValue=0.15, group=FeatherSim, parameterName=MaxFeatherRot}
    WindStrength=ScalarParameter() {defaultValue=0.015, group=FeatherSim, parameterName=WindStrength}
   ) {materialFunction=&/MF_FeatherSimulation_Core}.Normal) {name=VertexNormal, nodeColor=<rgba 0 0.0625 1 1>}
  NamedRerouteDeclaration(None=@n28) {name=Emissve, nodeColor=<rgba 1 0 0.3281 1>}
  NamedRerouteUsage()
  NamedRerouteUsage()
  NamedRerouteUsage()
  TextureSampleParameter2D() {
   parameterName=SpecMap
   samplerType=SAMPLERTYPE_Masks
   texture=/Game/Characters/Player/Female/Medium/Bodies/F_MED_TheHerald/Textures/T_F_MED_TheHerald_FaceAcc_S
  }
  TextureSample() {samplerType=SAMPLERTYPE_Masks, texture=/Game/Characters/Player/Female/Medium/Bodies/F_MED_TheHerald/Textures/T_F_MED_TheHerald_FaceAcc_M}
 ]
}
